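# Post-process-style AnimBlueprint for the Missing Link character's face accessory mesh, adding secondary motion and an emote-driven mouth.
# It copies the attached parent mesh's pose and curves, then constrains headTop between spine_05 and neck_01.
# Chained spring-based AnimDynamics nodes jiggle the bill, glasses, hat and two topknot bones.
# Each update switches the "Smile Mouth animation Frame" parameter on material slot 2 according to whether an emote plays.
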AnimBlueprint /BRCosmetics/Characters/Player/Male/Medium/Bodies/M_MED_Missing_Link/Meshes/Parts/M_MED_Missing_Link_FaceAcc_AnimBP
target_skeleton=&/M_MED_Missing_Link_FaceAcc_Skeleton
graphs [
 AnimGraph {nodes [
   n1 = "Constraint - headTop (PP)"!(
    ComponentPose=Local To Component(LocalPose=Copy Pose From Mesh() {bCopyCurves=true, bUseAttachedParent=true})
    Alpha=1
    ConstraintWeights_0=0.5
    ConstraintWeights_1:float=0.5
   ) -> (Pose) {
    alphaBoolBlend {blendOption=Linear}
    boneToModify {boneName=headTop}
    constraintSetup [{targetBone {boneName=spine_05}}, {targetBone {boneName=neck_01}}]
    constraintWeights [1,1]
   }
   n2 = Anim Dynamics - Bone: bill(ComponentPose=@n1, Alpha=1, ExternalForce="0,0,0") {
    angularDampingOverride=1
    angularSpringConstant=1500
    bAngularSpring=true
    bGravityOverrideInSimSpace=true
    bOverrideAngularDamping=true
    gravityScale=0.35
    simulationSpace=RootRelative
    alphaBoolBlend {blendOption=Linear}
    alphaScaleBiasClamp {scale=0.2}
    boundBone {boneName=bill}
    physicsBodyDefinitions [{
      sphereCollisionRadius=0
      boundBone {boneName=bill}
      boxExtents=<20 10 10>
      constraintSetup {angularLimitsMax=<5 10 0>, angularLimitsMin=<-5 -10 -40>, angularTarget=<0 0 0>}
      localJointOffset=<5 0 0>
     }]
   }
   n3 = Anim Dynamics - Bone: glasses(ComponentPose=@n2, Alpha=1, ExternalForce="0,0,0") {
    angularDampingOverride=1
    angularSpringConstant=1500
    bAngularSpring=true
    bGravityOverrideInSimSpace=true
    bOverrideAngularDamping=true
    gravityScale=0.35
    simulationSpace=RootRelative
    alphaBoolBlend {blendOption=Linear}
    alphaScaleBiasClamp {scale=0.12}
    boundBone {boneName=glasses}
    physicsBodyDefinitions [{
      sphereCollisionRadius=0
      boundBone {boneName=glasses}
      boxExtents=<20 10 20>
      constraintSetup {angularLimitsMax=<5 0 0>, angularLimitsMin=<-5 0 -40>, angularTarget=<1 0 0>}
      localJointOffset=<5 0 0>
     }]
   }
   n4 = Anim Dynamics - Bone: hat(ComponentPose=@n3, Alpha=1, ExternalForce="0,0,0") {
    angularDampingOverride=1
    angularSpringConstant=1500
    bAngularSpring=true
    bGravityOverrideInSimSpace=true
    bOverrideAngularDamping=true
    gravityScale=0.35
    simulationSpace=RootRelative
    alphaBoolBlend {blendOption=Linear}
    alphaScaleBiasClamp {scale=0.08}
    boundBone {boneName=hat}
    physicsBodyDefinitions [{
      sphereCollisionRadius=0
      boundBone {boneName=hat}
      boxExtents=<20 20 20>
      constraintSetup {angularLimitsMax=<0 20 20>, angularLimitsMin=<0 -20 -40>, angularTarget=<1 0 0>}
      localJointOffset=<5 0 0>
     }]
   }
   n5 = Anim Dynamics - Bone: topKnot_1(ComponentPose=@n4, Alpha=1, ExternalForce="0,0,0") {
    angularDampingOverride=1
    angularSpringConstant=1500
    bAngularSpring=true
    bGravityOverrideInSimSpace=true
    bOverrideAngularDamping=true
    gravityScale=0.35
    simulationSpace=RootRelative
    alphaBoolBlend {blendOption=Linear}
    alphaScaleBiasClamp {scale=0.15}
    boundBone {boneName=topKnot_1}
    componentLinearAccScale=<0.4 0.4 0.5>
    componentLinearVelScale=<0.4 0.4 0.5>
    physicsBodyDefinitions [{
      sphereCollisionRadius=0
      boundBone {boneName=topKnot_1}
      boxExtents=<20 20 10>
      constraintSetup {angularTargetAxis=AxisZ, twistAxis=AxisZ, angularLimitsMax=<50 50 0>, angularLimitsMin=<-50 -50 0>, angularTarget=<0 0 1>}
      localJointOffset=<0 0 -10>
     }]
   }
   n6 = Anim Dynamics - Bone: topKnot_2(ComponentPose=@n5, Alpha=1, ExternalForce="0,0,0") {
    angularDampingOverride=1
    angularSpringConstant=1500
    bAngularSpring=true
    bGravityOverrideInSimSpace=true
    bOverrideAngularDamping=true
    gravityScale=0.35
    simulationSpace=RootRelative
    alphaBoolBlend {blendOption=Linear}
    alphaScaleBiasClamp {scale=0.15}
    boundBone {boneName=topKnot_2}
    componentLinearAccScale=<0.4 0.4 0.5>
    componentLinearVelScale=<0.4 0.4 0.5>
    physicsBodyDefinitions [{
      sphereCollisionRadius=0
      boundBone {boneName=topKnot_2}
      boxExtents=<20 20 10>
      constraintSetup {angularTargetAxis=AxisZ, twistAxis=AxisZ, angularLimitsMax=<50 50 0>, angularLimitsMin=<-50 -50 0>, angularTarget=<0 0 1>}
      localJointOffset=<0 0 -10>
     }]
   }
   Output Pose(Result=Component To Local(ComponentPose=@n6)) {layerGroup=None}
  ]}
 EventGraph {nodes [
   Event BlueprintUpdateAnimation()
   > n1 = Get Dynamic Material Instance(SkeletalMeshComponent=GetPartSkeletalMeshComponent(), ElementIndex=2)
   > n2 = Branch(Condition=Get bIsPlayingEmote())
   > SetScalarParameterValue(self=@n1, ParameterName=Smile Mouth animation Frame, Value=8)
   SetScalarParameterValue(execute=@n2.else, self=@n1, ParameterName=Smile Mouth animation Frame, Value=1)
   TryGetPawnOwner()
  ]}
]
variables []
